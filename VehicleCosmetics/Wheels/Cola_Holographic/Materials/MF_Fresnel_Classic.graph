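# MF_Fresnel_Classic is a material function in the Cola_Holographic wheel cosmetic folder that computes a classic Fresnel term.
# It transforms the Normal input from tangent to world space and dots it with the camera vector.
# The result is inverted, saturated and raised to the Exponent input, giving values that rise toward grazing angles.
# Both inputs have preview defaults: a flat normal and an exponent of two.
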MaterialFunction /VehicleCosmetics/Wheels/Cola_Holographic/Materials/MF_Fresnel_Classic
graph {nodes [
  n1 = FunctionInput() {bUsePreviewValueAsDefault=true, inputName=Normal, previewValue=<0 0 1 1>}
  n2 = OneMinus(None=DotProduct(A=CameraVectorWS(), B=Transform(None=@n1)))
  n3 = FunctionInput() {bUsePreviewValueAsDefault=true, inputName=Exponent, inputType=FunctionInput_Scalar, sortPriority=1, previewValue=<2 0 0 1>}
  FunctionOutput(None=Power(Base=Saturate(None=@n2), Exp=@n3) {constExponent=3})
 ]}
