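# Unlit parent material for the Cola Holographic vehicle wheel, producing a pulsing, paintable emissive glow.
# MF_WheelParent_Simple supplies base attributes, then an emissive override feeds the vehicle damage and de-rez transition functions.
# Two time-driven sine pulses animate masked glow, with tire normal-map red and green channels faking shifting highlights.
# A Fresnel term adds edge glow before color blending with SilhouetteColor, which the Painted switch multiplies in again as a tint.
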Material /VehicleCosmetics/Wheels/Cola_Holographic/Materials/M_WheelParent_Cola_Holographic
graph {
 properties {
  bAutomaticallySetUsageInEditor=false
  bUseMaterialAttributes=true
  bUsedWithMeshDeformer=true
  bUsedWithSkeletalMesh=true
  pixelDepthOffsetMode=PDOM_Legacy
  previewMesh=/VehicleCosmetics/Wheels/Cola_Holographic/Meshes/SK_Wheel_ColaHolagraphic
  shadingModel=MSM_Unlit
  usedShadingModels=Unlit
 }
 nodes [
  n1 = MaterialFunctionCall() {materialFunction=&MF_WheelParent_Simple}
  n2 = NamedRerouteUsage()
  MaterialFunctionCall(In=MaterialFunctionCall(Material=SetMaterialAttributes!(MaterialAttributes=@n1.MaterialAttributes, Emissive Color=@n2, Base Color) {attributeSetTypes [#1, #2]}) {materialFunction=&MF_BaseVehicle_Damage}.Result) {materialFunction=&MF_VehicleDeRez_Transition}
  NamedRerouteDeclaration(None=VertexColor().R) {name=VC_R, nodeColor=<rgba 0 0.6719 1 1>}
  n3 = GetMaterialAttributes(None=@n1.MaterialAttributes) {attributeGetTypes [#3, #2]}
  NamedRerouteDeclaration(None=@n3.Normal) {name=Normal IN, nodeColor=<rgba 0.5 0 1 1>}
  NamedRerouteDeclaration(None=ComponentMask(None=@n1.RimRGBAMask) {r=true}) {name=RIM_M_R, nodeColor=<rgba 0.6875 0 1 1>}
  NamedRerouteDeclaration(None=ComponentMask(None=@n1.TireRGBAMask) {r=true}) {name=TIRE_M_R, nodeColor=<rgba 0 1 0.3672 1>}
  n4 = Multiply(A=Time()) {constB=0.5}
  n5 = Sine(None=@n4)
  n6 = ScalarParameter() {defaultValue=5, parameterName=WaveGlowIntensity}
  n7 = Multiply(A=Clamp(None=OneMinus(None=@n5)), B=@n6)
  n8 = Multiply(A=NamedRerouteUsage(), B=@n7)
  n9 = Multiply(A=Clamp(None=Sine(None=@n4)), B=@n6)
  n10 = Multiply(A=NamedRerouteUsage(), B=@n9)
  n11 = TextureSample() {samplerType=SAMPLERTYPE_Normal, texture=/VehicleCosmetics/Wheels/Cola_Holographic/Textures/T_ColaHolagraphic_Tire_N}
  n12 = Multiply(A=Abs(None=ComponentMask(None=@n11.RGB) {r=true}), B=@n9) {constB=5}
  n13 = Multiply(A=Clamp(None=ComponentMask(None=@n11.RGB) {g=true}), B=@n7) {constB=5}
  n14 = Max(A=@n10, B=Clamp(None=Max(A=@n12, B=@n13)))
  n15 = Clamp(None=Max(A=ConstantBiasScale(None=@n14) {bias=-0.1, scale=4}, B=@n10)) {maxDefault=5}
  n16 = NamedRerouteUsage()
  n17 = MaterialFunctionCall!(Normal=NamedRerouteUsage(), Exponent=ScalarParameter() {defaultValue=3, group=5_Special Effects, parameterName=Fresnel Intensity}) {materialFunction=&/MF_Fresnel_Classic}
  n18 = Multiply(A=ConstantBiasScale(None=NamedRerouteUsage())) {constB=5}
  n19 = Multiply(
   A=Max(A=LinearInterpolate(A=@n8, B=@n15, Alpha=@n16), B=Multiply(A=@n17.Result, B=@n18) {constB=5})
   B=ScalarParameter() {defaultValue=0.5, group=5_Special Effects, parameterName=EmissiveIntensity}
  )
  n20 = NamedRerouteUsage()
  n21 = LinearInterpolate(A=Multiply(A=NamedRerouteUsage()) {constB=10}, Alpha=@n20) {constB=0.277}
  n22 = VectorParameter() {group=5_Special Effects, parameterName=SilhouetteColor, defaultValue=<rgba 1 0 0 1>}
  n23 = LinearInterpolate(A=NamedRerouteUsage(), B=NamedRerouteUsage(), Alpha=@n20) {constB=0.277}
  n24 = Multiply(
   A=Clamp(None=ConstantBiasScale(None=@n19) {bias=0.02, scale=1}) {maxDefault=2}
   B=Clamp(None=LinearInterpolate(A=@n21, B=@n22.RGB, Alpha=@n23) {constB=0.277}) {maxDefault=2}
  )
  NamedRerouteDeclaration(None=StaticSwitchParameter(True=Multiply(A=@n24, B=@n22.RGB), False=@n24) {group=5_Special Effects, parameterName=Painted}) {name=Emissive, nodeColor=<rgba 0 0.4609 1 1>}
  NamedRerouteDeclaration(None=@n3.Base Color) {name=BC, nodeColor=<rgba 0 0.1562 1 1>}
 ]
}
